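; Weapon equip audio component for the Nitro Gauntlets that makes their looping sounds react to the player's movement.
; PlayValidEvent extends the parent, caching the resulting audio components when a played event matches the loop event tag.
; WeaponBindings subscribes to the instigating player pawn's SoundLibrary play events.
; When those events match movement tags, it fires the "Trigger.OnStep" audio parameter on each valid looping component, avoiding per-tick polling.
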(event EventBeginPlay
  (|CallParentFunction) ; This node is disabled and will not be called.\nDrag off pins to build functionality.
)

(event EventTick
  (params (DeltaSeconds :type Float%20%28single-precision%29)))

(fn OnPlayerSoundEvent
  (params (Results :type Sound%20Library%20Play%20Result%20Structure%20%28by%20ref%29) (Context :type Sound%20Library%20Context%20Object%20Reference))
  (if (GameplayTags|MatchesAnyTags (Utilities|Struct|BreakSoundLibraryPlayResult Results) (LocalVariables|GetMovementTags) true)
    (for _array_element (Variables|Private|GetLoopingAudioComponents)
      (Utilities|IsValid _array_element
        (:Is%20Valid
          (Audio|Parameter|ExecuteTriggerParameter _array_element "Trigger.OnStep"))))))

(fn PlayValidEvent
  (params (GameplayTag :type Gameplay%20Tag%20Structure))
  (bind _result (|CallParentFunction GameplayTag))
  (if (GameplayTags|MatchesTag (Utilities|Struct|BreakSoundLibraryPlayResult _result) (Variables|Event|GetLoopEventName) true)
    (Variables|Private|SetLoopingAudioComponents (Utilities|Struct|BreakSoundLibraryPlayResult _result))
    (else
      _))
  (return _result))

(fn WeaponBindings
  (|CallParentFunction)
  (bind _asfort_weapon (Utilities|Casting|CastToFortWeapon (Variables|Private|GetWeapon)))
  (bind _asfort_player_pawn_athena (Utilities|Casting|CastToFortPlayerPawnAthena (Variables|/Script/Engine.Actor|GetInstigator _asfort_weapon)))
  (EventDispatchers|BindEventtoOnSoundLibraryPlayEvent (Actor|GetComponentbyClass _asfort_player_pawn_athena "/Script/SoundLibrary.SoundLibraryComponent") (EventDispatchers|CreateEvent :!event OnPlayerSoundEvent)))
