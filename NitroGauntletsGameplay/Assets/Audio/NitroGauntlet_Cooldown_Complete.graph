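# NitroGauntlet_Cooldown_Complete is a Sound Cue for the Nitro Gauntlets cooldown-complete sound, routed through the Fort_UI sound class.
# A single wave player feeds the NitroGauntlet_Recharge_Ready_01 wave into a Modulator node at the root, which uses default settings.
# The cue plays at reduced volume and sets a very large max distance.
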SoundCue /NitroGauntletsGameplay/Assets/Audio/NitroGauntlet_Cooldown_Complete
root=Modulator_0
nodes [
 {id=Modulator_0, children [WavePlayer_1]}
 {id=WavePlayer_1, properties {soundWaveAssetPtr=&/NitroGauntlet_Recharge_Ready_01}}
]
properties {maxDistance=2097152, soundClassObject=&Fort_UI, volumeMultiplier=0.5}
