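# Sound cue for the Nitro Gauntlets' punch air launch.
# It mixes one of two random launch sounds at full level with one of three random in-air whooshes at slightly lower volume.
# A Modulator at the root varies the mixed result before playback through an attack-air attenuation and the player attacks and abilities sound class.
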SoundCue /NitroGauntletsGameplay/Assets/Audio/NitroGauntlet_Punch_Air_Launch
root=Modulator_0
nodes [
 {id=Mixer_1, children [Random_1,Random_0], properties {inputVolume [1,0.7]}}
 {id=Modulator_0, children [Mixer_1]}
 {id=Random_0, children [WavePlayer_0,WavePlayer_1,WavePlayer_2], properties {weights [1,1,1]}}
 {id=Random_1, children [WavePlayer_4,WavePlayer_5], properties {weights [1,1]}}
 {id=WavePlayer_0, properties {soundWaveAssetPtr=&/NitroGauntlet_InAir_Whoosh_01}}
 {id=WavePlayer_1, properties {soundWaveAssetPtr=&/NitroGauntlet_InAir_Whoosh_02}}
 {id=WavePlayer_2, properties {soundWaveAssetPtr=&/NitroGauntlet_InAir_Whoosh_03}}
 {id=WavePlayer_4, properties {soundWaveAssetPtr=&/NitroGauntlet_Punch_Air_Launch_01}}
 {id=WavePlayer_5, properties {soundWaveAssetPtr=&/NitroGauntlet_Punch_Air_Launch_02}}
]
properties {
 attenuationSettings=&/ATT_NitroGauntlets_Attack_Air
 maxDistance=10500
 soundClassObject=/Game/Sounds/FortSoundClassesAndModes/Fort_06_Player_Attacks_Abilities
 volumeMultiplier=1
}
