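# Sound cue for the looping in-air sound of the Nitro Gauntlets punch, classed under player attacks and abilities.
# A looping wave player feeds an Enveloper whose volume curve fades the loop in from silence over about a second, rising slowly at first and then more steeply.
# A dedicated Nitro Gauntlets air-attack attenuation asset handles spatial falloff.
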SoundCue /NitroGauntletsGameplay/Assets/Audio/NitroGauntlet_Punch_Air_Loop
root=Enveloper_0
nodes [
 {
  id=Enveloper_0
  children [WavePlayer_3]
  properties {pitchCurve {editorCurveData {keys [{time=0}]}}, volumeCurve {editorCurveData {keys [<at 0 0>,<at 0.8022 0.2492>,<at 1.2 1>]}}}
 }
 {id=WavePlayer_3, properties {bLooping=true, soundWaveAssetPtr=&/NitroGauntlet_InAir_Loop_01}}
]
properties {
 attenuationSettings=&/ATT_NitroGauntlets_Attack_Air
 maxDistance=10500
 soundClassObject=/Game/Sounds/FortSoundClassesAndModes/Fort_06_Player_Attacks_Abilities
 volumeMultiplier=1
}
