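# Sound cue for the Nitro Gauntlets air punch that hits nothing, routed to the player attacks and abilities sound class.
# A single wave player feeds a modulator node at reduced cue volume, spatialized through the Nitro Gauntlets air attack attenuation asset.
# Its own concurrency override allows one instance at a time, stopping the oldest, with a short retrigger window.
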SoundCue /NitroGauntletsGameplay/Assets/Audio/NitroGauntlet_Punch_Air_NoImpact
root=Modulator_0
nodes [
 {id=Modulator_0, children [WavePlayer_0]}
 {id=WavePlayer_0, properties {soundWaveAssetPtr=&/NitroGauntlet_Punch_Air_NoImpact_01}}
]
properties {
 attenuationSettings=&/ATT_NitroGauntlets_Attack_Air
 bOverrideConcurrency=true
 maxDistance=10500
 soundClassObject=/Game/Sounds/FortSoundClassesAndModes/Fort_06_Player_Attacks_Abilities
 volumeMultiplier=0.7
 concurrencyOverrides {maxCount=1, resolutionRule=StopOldest, retriggerTime=0.1}
}
