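# Sound cue for the Nitro Gauntlets air slam punch, in the variant without debris.
# A Random node picks between two wave variations with equal weight, then passes the result through a Modulator.
# It plays in the player attacks and abilities sound class with the Nitro Gauntlets air attack attenuation asset.
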SoundCue /NitroGauntletsGameplay/Assets/Audio/NitroGauntlet_Punch_Air_Slam_NoDebris
root=Modulator_0
nodes [
 {id=Modulator_0, children [Random_0]}
 {id=Random_0, children [WavePlayer_0,WavePlayer_1], properties {weights [1,1]}}
 {id=WavePlayer_0, properties {soundWaveAssetPtr=&/NitroGauntlet_Punch_Air_Slam_NoDebris_01}}
 {id=WavePlayer_1, properties {soundWaveAssetPtr=&/NitroGauntlet_Punch_Air_Slam_NoDebris_02}}
]
properties {
 attenuationSettings=&/ATT_NitroGauntlets_Attack_Air
 maxDistance=10500
 soundClassObject=/Game/Sounds/FortSoundClassesAndModes/Fort_06_Player_Attacks_Abilities
 volumeMultiplier=1
}
