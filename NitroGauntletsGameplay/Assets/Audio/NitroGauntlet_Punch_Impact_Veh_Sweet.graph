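# Sound cue for the "sweet" layer of a Nitro Gauntlets punch impact on a vehicle, judging by its name.
# A Random node picks one of five equally weighted wave variations and feeds it through a Modulator at the root.
# It uses the gauntlets' ground-attack attenuation and the player attacks and abilities sound class.
# A concurrency override allows one instance, stopping the oldest, with a short retrigger time.
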SoundCue /NitroGauntletsGameplay/Assets/Audio/NitroGauntlet_Punch_Impact_Veh_Sweet
root=Modulator_0
nodes [
 {id=Modulator_0, children [Random_1]}
 {id=Random_1, children [WavePlayer_10, WavePlayer_11, WavePlayer_12, WavePlayer_13, WavePlayer_14], properties {weights [1,1,1,1,1]}}
 {id=WavePlayer_10, properties {soundWaveAssetPtr=&/NitroGauntlet_Punch_Impact_Veh_Sweet_01}}
 {id=WavePlayer_11, properties {soundWaveAssetPtr=&/NitroGauntlet_Punch_Impact_Veh_Sweet_02}}
 {id=WavePlayer_12, properties {soundWaveAssetPtr=&/NitroGauntlet_Punch_Impact_Veh_Sweet_03}}
 {id=WavePlayer_13, properties {soundWaveAssetPtr=&/NitroGauntlet_Punch_Impact_Veh_Sweet_04}}
 {id=WavePlayer_14, properties {soundWaveAssetPtr=&/NitroGauntlet_Punch_Impact_Veh_Sweet_05}}
]
properties {
 attenuationSettings=&/ATT_NitroGauntlets_Attack_Ground
 bOverrideConcurrency=true
 maxDistance=6400
 soundClassObject=/Game/Sounds/FortSoundClassesAndModes/Fort_06_Player_Attacks_Abilities
 volumeMultiplier=0.7
 concurrencyOverrides {maxCount=1, resolutionRule=StopOldest, retriggerTime=0.1}
}
