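# Sound cue for the Nitro Gauntlet "L" punch, classed under player attacks and abilities.
# A Random node picks one of four equally weighted punch wave variants.
# Its output feeds a Modulator at default settings, which varies pitch and volume on each play.
# Spatial falloff comes from the shared Nitro Gauntlets ground attack attenuation asset.
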SoundCue /NitroGauntletsGameplay/Assets/Audio/NitroGauntlet_Punch_L
root=Modulator_0
nodes [
 {id=Modulator_0, children [Random_0]}
 {id=Random_0, children [WavePlayer_0,WavePlayer_1,WavePlayer_2,WavePlayer_3], properties {weights [1,1,1,1]}}
 {id=WavePlayer_0, properties {soundWaveAssetPtr=&/NitroGauntlet_Punch_L_01}}
 {id=WavePlayer_1, properties {soundWaveAssetPtr=&/NitroGauntlet_Punch_L_02}}
 {id=WavePlayer_2, properties {soundWaveAssetPtr=&/NitroGauntlet_Punch_L_03}}
 {id=WavePlayer_3, properties {soundWaveAssetPtr=&/NitroGauntlet_Punch_L_04}}
]
properties {
 attenuationSettings=&/ATT_NitroGauntlets_Attack_Ground
 maxDistance=6400
 soundClassObject=/Game/Sounds/FortSoundClassesAndModes/Fort_06_Player_Attacks_Abilities
 volumeMultiplier=1
}
