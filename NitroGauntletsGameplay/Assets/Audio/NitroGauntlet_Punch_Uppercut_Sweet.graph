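# Sound cue for the Nitro Gauntlets uppercut punch sweetener, playing an air uppercut shockwave sound.
# A single wave player feeds a modulator left at its default settings for pitch and volume variation.
# The cue uses the gauntlets' distant air-attack attenuation and routes to the player attacks and abilities sound class.
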SoundCue /NitroGauntletsGameplay/Assets/Audio/NitroGauntlet_Punch_Uppercut_Sweet
root=Modulator_0
nodes [
 {id=Modulator_0, children [WavePlayer_2]}
 {id=WavePlayer_2, properties {soundWaveAssetPtr=&/NitroGauntlet_Punch_Air_Uppercut_Shockwave_01}}
]
properties {
 attenuationSettings=&/ATT_NitroGauntlets_Attack_Air_Dist
 maxDistance=15500
 soundClassObject=/Game/Sounds/FortSoundClassesAndModes/Fort_06_Player_Attacks_Abilities
 volumeMultiplier=1
}
