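# Niagara ribbon trail effect for the Nitro Gauntlets air sprint.
# A single emitter spawns short-lived particles per unit of movement, split into eight partitions whose index selects the bone or socket sampled on the skeletal mesh.
# Curves drive ribbon width, fade alpha out and cool the color from HDR orange to black over each particle's life.
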NiagaraSystem /NitroGauntletsGameplay/Assets/FX/NS_NitroGauntlets_AirSprint_Trail
dependencies {
 usedDataInterfaces [&NiagaraDataInterfaceCurve, &NiagaraDataInterfaceColorCurve, &NiagaraDataInterfaceSkeletalMesh]
 usedDynamicInputs []
 usedModules [
  &self:SystemSpawnScript.SystemScriptSource.SystemScriptGraph.NiagaraNodeAssignment_0.SetVariables_#1
  &SystemState
 ]
}
emitters {Trial {
  emitterUpdateScript [
   &EmitterState()
   &SpawnPerUnit(Module.Use Max Movement Threshold=false, Module.Use Movement Tolerance=false, Spawn Spacing=6, Use Spawn Probability=false)
  ]
  particleSpawnScript [
   &PartitionParticles(Number Of Partitions=8)
   &InitializeParticle(Lifetime=0.2, Ribbon Width=30, Ribbon Width Mode=::NewEnumerator1, UsePositionOffset=false, Write Material Random=true)
   &SkeletalMeshLocation(
    Bone / Socket Index=$Particles.PartitionParticles.Partition
    Bone Sampling Mode=::NewEnumerator4
    MaskSpawnGroup=false
    Skeletal Mesh=$System.Skeletal Mesh
   )
  ]
  particleUpdateScript [
   &ParticleState()
   &ScaleRibbonWidth(Ribbon Width Scale=&FloatFromCurve(FloatCurve=DataInterface() {propertyValues {Curve {keys [{arriveTangent=-1.1994, interpMode=RCIM_Cubic, time=0, value=2}, <at 0.8 2>]}}}))
   &SolveForcesAndVelocity(Clamp Velocity=false, Limit Acceleration=false)
   &ScaleColor(
    Scale Alpha=&FloatFromCurve(FloatCurve=DataInterface() {propertyValues {Curve {keys [{interpMode=RCIM_Cubic, tangentMode=RCTM_User, time=0, value=1}, {interpMode=RCIM_Cubic, tangentMode=RCTM_User, time=1, value=0}]}}})
    Scale RGB=&MakeVectorFromLinearColorRGB(Color=&ColorFromCurve(ColorCurve=DataInterface() {propertyValues {
        AlphaCurve {keys []}
        BlueCurve {keys [<at 0.0461 0.5085>,<at 0.1631 0.1521>,<at 0.3304 0.1219>,<at 0.5067 0>]}
        GreenCurve {keys [<at 0.0461 2.3718>,<at 0.1631 0.3177>,<at 0.3304 0.1412>,<at 0.5067 0>]}
        RedCurve {keys [<at 0.0461 10>,<at 0.1631 2>,<at 0.3304 0.1948>,<at 0.5067 0>]}
       }}))
    ScaleA=true
    ScaleRGB=true
   ) {module=ScaleColor001}
  ]
  rendererClasses [&NiagaraRibbonRendererProperties]
  renderers [{rendererClass=&NiagaraRibbonRendererProperties}]
 }}
summary {
 systemName=NS_NitroGauntlets_AirSprint_Trail
 emitters [{emitterName=Trial, rendererClasses [&NiagaraRibbonRendererProperties]}]
 userVariables [User.LocalView {type=NiagaraBool, defaultValue=true}]
}
